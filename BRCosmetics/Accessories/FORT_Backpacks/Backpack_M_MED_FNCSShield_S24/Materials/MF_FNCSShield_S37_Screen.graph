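# Material function for the FNCS Shield back bling screen: a score display with startup logo and celebration graphics, output as material attributes, screen mask and Fresnel.
# Score selects single- or double-digit layout and offsets UVs into a digit font texture.
# A Timestamp transition, or looping debug time, drives Jumbotron curve-atlas rows that animate digits, flashes, sweep, pixelated rings, burst, ribbons and streaks.
# It also builds CRT scanlines, periodic glitch distortion and switchable backgrounds.
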MaterialFunction /BRCosmetics/Accessories/FORT_Backpacks/Backpack_M_MED_FNCSShield_S24/Materials/MF_FNCSShield_S37_Screen
graph {nodes [
  n1 = TextureCoordinate() {coordinateIndex=1}
  n2 = TextureSampleParameter2D(UVs=@n1) {
   group=FNCSShield_Screen
   parameterName=Screen Mask
   samplerType=SAMPLERTYPE_Masks
   sortPriority=10
   texture=&T_BlackMask
  }
  FunctionOutput(None=@n2.RGB) {outputName=M, sortPriority=1}
  n3 = SetMaterialAttributes!(
   MaterialAttributes
   BaseColor=Constant()
   EmissiveColor=LinearInterpolate(A=NamedRerouteUsage(), B=NamedRerouteUsage(), Alpha=NamedRerouteUsage())
   Metallic=Constant() {r=0.84}
   Specular=Constant() {r=0.84}
   Roughness=ScalarParameter() {defaultValue=0.15, group=FNCSShield_Screen, parameterName=ScreenRoughness, sortPriority=11}
   Normal=NamedRerouteUsage()
   AmbientOcclusion=NamedRerouteUsage()
  ) {attributeSetTypes [
    #1
    #2
    #3
    #4
    #5
    #6
    #7
   ]}
  FunctionOutput(None=@n3) {outputName="Material (MA)"}
  n4 = LinearInterpolate(
   A=ScalarParameter() {defaultValue=0.55, group=FNCSShield_Screen, parameterName=ScreenFresnelLow, sortPriority=30}
   Alpha=MaterialFunctionCall(Exponent=ScalarParameter() {defaultValue=2, group=FNCSShield_Screen, parameterName=ScreenFresnelExponent, sortPriority=30}) {materialFunction=&MF_Fresnel_FX}.Fresnel
  )
  FunctionOutput(None=@n4) {outputName=Fresnel, sortPriority=3}
  n5 = NamedRerouteUsage()
  n6 = VectorParameter() {group=FNCSShield_Screen_Graphics, parameterName=CelebrationBurstColorMid, sortPriority=20, defaultValue=<rgba 1 0.5 0 0>}
  n7 = VectorParameter() {group=FNCSShield_Screen_Graphics, parameterName=CelebrationBurstColorHigh, sortPriority=20, defaultValue=<rgba 1 0.5 0 0>}
  n8 = NamedRerouteUsage()
  n9 = MaterialFunctionCall(A=@n5, B=@n6.RGB, C=@n7.RGB, Alpha=@n8) {materialFunction=&Lerp_3Color}
  LinearInterpolate(
   A=@n5
   B=@n9.Result
   Alpha=Step(Y=@n8, X=ScalarParameter() {defaultValue=1, group=FNCSShield_Screen_Graphics, parameterName=CelebrationBurstCutoffValue})
  )
  MaterialFunctionCall!(
   Time
   Divisor=ScalarParameter() {defaultValue=1.5, group=FNCSShield_Screen, parameterName=DebugTimeDivisor, sortPriority=30}
   Mod=ScalarParameter() {defaultValue=2, group=FNCSShield_Screen, parameterName=DebugTimeModulo, sortPriority=30}
  ) {materialFunction=/Game/Effects/Fort_Effects/Materials/_Functions/MF_ModuloTime}
  n10 = VectorParameter() {
   group=FNCSShield_Screen
   parameterName=UV_Mask
   sortPriority=20
   channelNames {a=Double_Bias, b=Double_Divisor, g=Single_Bias, r=Single_Divisor}
   defaultValue=<rgba 2.6 -0.2987 2.8 -0.578>
  }
  n11 = MaterialFunctionCall(Float2=LinearInterpolate(
    A=MaterialFunctionCall(X=@n10.Single_Divisor, Y=@n10.Single_Divisor) {materialFunction=&MakeFloat2}.Result
    B=MaterialFunctionCall(X=@n10.Double_Divisor, Y=@n10.Double_Divisor) {materialFunction=&MakeFloat2}.Result
    Alpha=NamedRerouteUsage()
   )) {materialFunction=&BreakOutFloat2Components}
  n12 = MaterialFunctionCall(Float2=NamedRerouteUsage()) {materialFunction=&BreakOutFloat2Components}
  n13 = Add(A=@n11.G, B=@n12.R)
  n14 = Divide(A=NamedRerouteUsage(), B=@n11.R)
  n15 = Constant() {r=1}
  n16 = Divide(A=NamedRerouteUsage())
  n17 = Add(A=@n16, B=Divide(A=@n16) {constB=-20, desc=?}) {constB=-0.01}
  n18 = Divide(A=@n17)
  NamedRerouteDeclaration(None=Min(A=SphereMask(A=@n13, B=@n14, Radius=@n14, Hardness=@n15), B=SphereMask(A=@n12.G, B=@n17, Radius=@n18, Hardness=@n15))) {name=NumbersBoxMask, nodeColor=<rgba 0 1 0.8828 1>}
  n19 = MaterialFunctionCall(Float2=NamedRerouteUsage()) {materialFunction=&BreakOutFloat2Components}
  n20 = CurveAtlasRowParameter(CurveTime=NamedRerouteUsage()) {
   atlas=/BRCosmetics/Accessories/FORT_Backpacks/Backpack_M_MED_Jumbotron_S20/Textures/CA_JumboTron
   curve=/Game/Accessories/FORT_Backpacks/Backpack_M_MED_Jumbotron_S20/Textures/C_JumboTron_Sweep
   defaultValue=5
   group=FNCSShield_Screen_GraphicsCurve
   parameterName=Sweep
   sortPriority=0
  }
  n21 = Add(A=@n20.R) {constB=-1, desc=Curves can't go negative . . .}
  n22 = Constant() {r=0.9}
  n23 = Constant()
  NamedRerouteDeclaration(None=Power(Base=SphereMask(A=@n19.R, B=@n21, Radius=@n22, Hardness=@n23)) {constExponent=2.5}) {name=Sweep, nodeColor=<rgba 0.3594 0 1 1>}
  n24 = ScalarParameter() {group=FNCSShield_Screen, parameterName=Score, sliderMax=99, sortPriority=0}
  NamedRerouteDeclaration(None=Step(X=@n24) {constX=10, constY=10}) {name=IfScoreOverTen, nodeColor=<rgba 0.2188 1 0 1>}
  n25 = ScalarParameter() {defaultValue=16, group=FNCSShield_Screen_Graphics, parameterName=RadialEffect_Pixelation, sortPriority=30}
  n26 = Ceil(None=Multiply(A=NamedRerouteUsage(), B=@n25))
  n27 = CurveAtlasRowParameter(CurveTime=NamedRerouteUsage()) {
   atlas=&CA_JumboTron
   curve=/Game/Accessories/FORT_Backpacks/Backpack_M_MED_Jumbotron_S20/Textures/C_Jumbotron_RadialEffect_S21
   defaultValue=6
   group=FNCSShield_Screen_GraphicsCurve
   parameterName=RadialEffect
   sortPriority=0
  }
  n28 = ScalarParameter() {defaultValue=8, group=FNCSShield_Screen_Graphics, parameterName=RadialEffect_PulseRate, sortPriority=30}
  NamedRerouteDeclaration(None=MaterialFunctionCall(UVs=Divide(A=@n26, B=@n25), Texture Scale=Multiply(A=@n27, B=@n28)) {materialFunction=&ScaleUVsByCenter}.UVs) {name=RadialUV, nodeColor=<rgba 1 0 0.3984 1>}
  NamedRerouteDeclaration(None=@n2.R) {name=outAO, nodeColor=<rgba 1 0.1641 0 1>}
  n29 = NamedRerouteUsage()
  n30 = ScalarParameter() {defaultValue=0.2, group=FNCSShield_Screen_Graphics, parameterName=Outer Radius A, sortPriority=30}
  n31 = Constant() {r=100}
  n32 = ScalarParameter() {defaultValue=0.18, group=FNCSShield_Screen_Graphics, parameterName=Inner Radius A, sortPriority=30}
  n33 = NamedRerouteUsage()
  n34 = ScalarParameter() {defaultValue=0.14, group=FNCSShield_Screen_Graphics, parameterName=Outer Radius B, sortPriority=30}
  n35 = Constant() {r=100}
  n36 = ScalarParameter() {defaultValue=0.13, group=FNCSShield_Screen_Graphics, parameterName=Inner Radius B, sortPriority=30}
  n37 = Add(
   A=Subtract(
    A=MaterialFunctionCall(UVs=@n29, Radius=@n30, Density=@n31) {desc=Outer Circle, materialFunction=&RadialGradientExponential}.RadialGradientExponential
    B=MaterialFunctionCall(UVs=@n29, Radius=@n32, Density=@n31) {desc=Inner Circle, materialFunction=&RadialGradientExponential}.RadialGradientExponential
   )
   B=Subtract(
    A=MaterialFunctionCall(UVs=@n33, Radius=@n34, Density=@n35) {desc=Outer Circle, materialFunction=&RadialGradientExponential}.RadialGradientExponential
    B=MaterialFunctionCall(UVs=@n33, Radius=@n36, Density=@n35) {desc=Inner Circle, materialFunction=&RadialGradientExponential}.RadialGradientExponential
   )
  )
  n38 = NamedRerouteUsage()
  n39 = Constant() {r=100}
  n40 = Subtract(
   A=MaterialFunctionCall(
    UVs=@n38
    Radius=ScalarParameter() {defaultValue=0.1, group=FNCSShield_Screen_Graphics, parameterName=Outer Radius C, sortPriority=30}
    Density=@n39
   ) {desc=Outer Circle, materialFunction=&RadialGradientExponential}.RadialGradientExponential
   B=MaterialFunctionCall(
    UVs=@n38
    Radius=ScalarParameter() {defaultValue=0.09, group=FNCSShield_Screen_Graphics, parameterName=Inner Radius C, sortPriority=30}
    Density=@n39
   ) {desc=Inner Circle, materialFunction=&RadialGradientExponential}.RadialGradientExponential
  )
  NamedRerouteDeclaration(None=Multiply(
    A=Add(A=@n37, B=@n40)
    B=ScalarParameter() {defaultValue=0.9, group=FNCSShield_Screen_Graphics, parameterName=RadialEffect_Opacity, sliderMax=1}
   ) {constB=0.05}) {name=RadialEffect, nodeColor=<rgba 1 0 0.1935 1>}
  n41 = VectorParameter() {group=FNCSShield_Screen_Graphics, parameterName=TextColor, sortPriority=20, defaultValue=<rgba 1 1 1 1>}
  n42 = VectorParameter() {group=FNCSShield_Screen_Graphics, parameterName=TextColorCelebrate, sortPriority=20, defaultValue=<rgba 1 1 0 1>}
  n43 = Sign(None=Sine(None=Time()) {period=0.25})
  n44 = CurveAtlasRowParameter(CurveTime=NamedRerouteUsage()) {
   atlas=&CA_JumboTron
   curve=/Game/Accessories/FORT_Backpacks/Backpack_M_MED_Jumbotron_S20/Textures/C_JumboTron_Flashes
   defaultValue=4
   group=FNCSShield_Screen_GraphicsCurve
   parameterName=Flashes
   sortPriority=0
  }
  n45 = LinearInterpolate(A=@n41.RGB, B=@n42.RGB, Alpha=Multiply(A=Saturate(None=@n43), B=@n44.R))
  n46 = NamedRerouteUsage()
  n47 = VectorParameter() {group=FNCSShield_Screen_Graphics, parameterName=RadialEffectColor, sortPriority=20, defaultValue=<rgba 1 1 0 1>}
  n48 = LinearInterpolate(A=@n9.Result, B=@n47.RGB, Alpha=NamedRerouteUsage())
  n49 = VectorParameter() {group=FNCSShield_Screen_Graphics, parameterName=CelebrationStreaksColor, sortPriority=20, defaultValue=<rgba 1 0.5 0 1>}
  n50 = StaticSwitchParameter(True=LinearInterpolate(A=@n48, B=@n49.RGB, Alpha=NamedRerouteUsage()), False=@n48) {group=FNCSShield_Screen_Graphics, parameterName=EnableCelebrationStreaks, sortPriority=0}
  n51 = VectorParameter() {group=FNCSShield_Screen_Graphics, parameterName=CelebrationRibbonColor, sortPriority=20, defaultValue=<rgba 1 0.2 0 1>}
  n52 = StaticSwitchParameter(True=LinearInterpolate(A=@n50, B=@n51.RGB, Alpha=NamedRerouteUsage()), False=@n50) {group=FNCSShield_Screen_Graphics, parameterName=EnableCelebrationRibbon, sortPriority=0}
  n53 = VectorParameter() {group=FNCSShield_Screen_Graphics, parameterName=TextShadowColor, sortPriority=20, defaultValue=<rgba 0.1 0.1 0.1 1>}
  n54 = LinearInterpolate(
   A=@n52
   B=LinearInterpolate(A=LinearInterpolate(A=@n52, B=@n45, Alpha=@n46), B=@n53.RGB, Alpha=NamedRerouteUsage())
   Alpha=NamedRerouteUsage()
  )
  n55 = VectorParameter() {group=FNCSShield_Screen_Graphics, parameterName=SweepColor, sortPriority=20, defaultValue=<rgba 0 2 2 1>}
  n56 = NamedRerouteUsage()
  NamedRerouteDeclaration(None=Multiply(A=LinearInterpolate(A=@n54, B=@n55.RGB, Alpha=@n56), B=@n4)) {name=outFinalScreen, nodeColor=<rgba 0.5938 0 1 1>}
  n57 = CurveAtlasRowParameter(CurveTime=NamedRerouteUsage()) {
   atlas=&CA_JumboTron
   curve=/Game/Accessories/FORT_Backpacks/Backpack_M_MED_Jumbotron_S20/Textures/C_JumboTron_Streaks
   defaultValue=2
   group=FNCSShield_Screen_GraphicsCurve
   parameterName=Streaks
   sortPriority=0
  }
  n58 = MaterialFunctionCall(UVs=NamedRerouteUsage(), Texture Scale=@n57.R) {materialFunction=&ScaleUVsByCenter}
  n59 = TextureObjectParameter() {
   group=FNCSShield_Screen_Graphics
   parameterName=CelebrationGraphs
   samplerType=SAMPLERTYPE_Masks
   sortPriority=10
   texture=&T_BlackMask
  }
  n60 = TextureSample(UVs=Rotator(Coordinate=@n58.UVs) {speed=0.944}, Tex=@n59) {samplerSource=SSM_Clamp_WorldGroupSettings, samplerType=SAMPLERTYPE_LinearColor}
  n61 = TextureSample(UVs=Rotator(Coordinate=@n58.UVs) {speed=-0.52}, Tex=@n59) {samplerSource=SSM_Clamp_WorldGroupSettings, samplerType=SAMPLERTYPE_LinearColor}
  n62 = MaterialFunctionCall(UVs=NamedRerouteUsage(), Texture Scale=Constant() {r=1.1}) {materialFunction=&ScaleUVsByCenter}
  n63 = Max(
   A=TextureSample(UVs=Rotator(Coordinate=@n62.UVs) {speed=-0.138}, Tex=@n59) {samplerSource=SSM_Clamp_WorldGroupSettings, samplerType=SAMPLERTYPE_LinearColor}.R
   B=TextureSample(UVs=Rotator(Coordinate=@n62.UVs) {speed=0.1239}, Tex=@n59) {samplerSource=SSM_Clamp_WorldGroupSettings, samplerType=SAMPLERTYPE_LinearColor}.R
  )
  NamedRerouteDeclaration(None=Max(A=Max(A=@n60.R, B=@n61.R), B=Multiply(A=@n63) {constB=0.8})) {name=CelebrationStreaks, nodeColor=<rgba 1 0.7266 0 1>}
  NamedRerouteDeclaration(None=Multiply(A=@n52, B=@n4)) {name=outBackgroundOnly, nodeColor=<rgba 1 0.8906 0 1>}
  n64 = Multiply(A=NamedRerouteUsage(), B=ScalarParameter() {defaultValue=50, group=FNCSShield_Screen, parameterName=CRT_UVScale, sortPriority=30}) {constB=5}
  n65 = SmoothStep(Value=ComponentMask(None=Frac(None=@n64)) {g=true}) {constMax=0.5, constMin=0.5}
  n66 = VectorParameter() {group=FNCSShield_Screen, parameterName=CRT_Scanline_Color, sortPriority=20, defaultValue=<rgba 1 1 1 1>}
  NamedRerouteDeclaration(None=Multiply(
    A=Multiply(A=@n65, B=@n66.RGB)
    B=ScalarParameter() {defaultValue=1.5, group=FNCSShield_Screen, parameterName=CRT_ColorScale, sortPriority=30}
   )) {name=CRTPattern, nodeColor=<rgba 1 0 0.3047 1>}
  n67 = TextureObjectParameter() {
   group=FNCSShield_Screen_Graphics
   parameterName=DigitsFont
   samplerType=SAMPLERTYPE_Masks
   sortPriority=10
   texture=&T_BlackMask
  }
  n68 = NamedRerouteUsage()
  n69 = TextureSample!(UVs=NamedRerouteUsage(), Tex=@n67, "DDX(UVs)"=DDX(Value=@n68), "DDY(UVs)"=DDY(Value=@n68), Apply View MipBias) {mipValueMode=TMVM_Derivative, samplerSource=SSM_Wrap_WorldGroupSettings, samplerType=SAMPLERTYPE_LinearColor}
  NamedRerouteDeclaration(None=@n69.R) {name=RedChannel Digits, nodeColor=<rgba 0 1 0.9297 1>}
  NamedRerouteDeclaration(None=@n69.G) {name=GreenChannel Digits, nodeColor=<rgba 0 1 0 1>}
  n70 = ComponentMask(None=VectorParameter() {group=FNCSShield_Screen, parameterName=UVs_Offset, sortPriority=20, defaultValue=<rgba 0 0 0 1>}.RGB) {g=true, r=true}
  NamedRerouteDeclaration(None=@n70) {name=uvoffset, nodeColor=<rgba 0 1 0.7891 1>}
  n71 = NamedRerouteUsage() {desc=Layer 0}
  n72 = VectorParameter() {group=FNCSShield_Screen, parameterName=FNCS Logo Shadow Color, sortPriority=20, defaultValue=<rgba 0 0 0 0.5>}
  n73 = NamedRerouteUsage()
  n74 = TextureSampleParameter2D(UVs=@n73) {
   group=FNCSShield_Screen
   parameterName=Startup Screen Mask
   samplerType=SAMPLERTYPE_Masks
   sortPriority=10
   texture=&T_BlackMask
  }
  n75 = Multiply(A=Multiply(A=@n72.RGB, B=@n72.RGB), B=@n74.B) {desc=Colored Shadow}
  n76 = Add(A=Add(A=@n74.R, B=@n74.R), B=@n74.B)
  n77 = VectorParameter() {group=FNCSShield_Screen, parameterName=FNCS Logo Color, sortPriority=20, defaultValue=<rgba 1 1 1 1>}
  n78 = TextureSampleParameter2D(UVs=@n73) {
   group=FNCSShield_Screen
   parameterName=Startup Screen Texture
   samplerType=SAMPLERTYPE_Masks
   sortPriority=10
   texture=/Game/Global/Textures/Default/Blanks/T_White
  }
  n79 = StaticSwitchParameter(
   True=LinearInterpolate(A=LinearInterpolate(A=@n71, B=@n75, Alpha=@n76) {desc=Layer 1}, B=Multiply(A=@n77.RGB, B=@n74.R) {desc=Colored Logo}, Alpha=@n74.R) {desc=Layer 2}
   False=@n78.RGB
  ) {defaultValue=true, desc=True: Mask False: Texture, group=FNCSShield_Screen, parameterName=Mask as Startup Screen Graph?, sortPriority=2}
  NamedRerouteDeclaration(None=@n79) {name=Startup Screen, nodeColor=<rgba 0 0.4964 1 1>}
  n80 = MaterialFunctionCall(Divisor=Constant() {r=0.333}, Mod=Constant() {r=1000}) {materialFunction=&MF_ModuloTime}
  n81 = ComponentMask(None=MaterialFunctionCall!(
    PrimaryTextureCoordinates=TextureCoordinate() {coordinateIndex=1, vTiling=50}
    2ndOrderTextureCoordinates
    RandomSeed=Round(None=@n80.Unclamped)
    NoiseDensityBias
    Bypass Noise Influence Crosstalk
   ) {materialFunction=/Game/Effects/Fort_Effects/Materials/_Functions/MF_RandomUVSample}.NoiseWithRange) {g=true}
  n82 = Constant()
  n83 = Time()
  n84 = Power(Base=Max(A=Sine(None=@n83) {period=3.6}, B=Sine(None=@n83) {period=2.44})) {constExponent=20}
  n85 = Add(A=Multiply(A=AppendVector(A=@n81, B=@n82), B=Round(None=@n84))) {constB=-0.5}
  n86 = ScalarParameter() {defaultValue=0.06, group=FNCSShield_Screen, parameterName=GlitchScale, sortPriority=30}
  NamedRerouteDeclaration(None=Saturate(None=Multiply(A=@n85, B=@n86))) {name=GlitchDistortion, nodeColor=<rgba 1 0.6562 0 1>}
  n87 = Add(A=TextureCoordinate() {coordinateIndex=1}, B=@n70)
  n88 = NamedRerouteUsage()
  n89 = Rotator(
   Coordinate=NamedRerouteDeclaration(None=Add(A=@n87, B=@n88)) {name=UVs, nodeColor=<rgba 0.4531 0 1 1>}
   Time=Constant() {desc=0.4 As Default, r=0.4}
  )
  n90 = CurveAtlasRowParameter(CurveTime=NamedRerouteUsage()) {
   atlas=&CA_JumboTron
   curve=/Game/Accessories/FORT_Backpacks/Backpack_M_MED_Jumbotron_S20/Textures/C_JumboTron_Numbers
   defaultValue=3
   group=FNCSShield_Screen_GraphicsCurve
   parameterName=Numbers
   sortPriority=0
  }
  n91 = VectorParameter() {
   group=FNCSShield_Screen
   parameterName=UV_Scale
   sortPriority=20
   channelNames {a=Double_V_Scale, b=Double_U_Scale, g=Single_V_Scale, r=Single_U_Scale}
   defaultValue=<rgba 2.5811 0.15 4.0696 0.22>
  }
  n92 = MaterialFunctionCall(Float2=LinearInterpolate(
    A=MaterialFunctionCall(X=@n91.Single_U_Scale, Y=@n91.Single_U_Scale) {materialFunction=&MakeFloat2}.Result
    B=MaterialFunctionCall(X=@n91.Double_U_Scale, Y=@n91.Double_U_Scale) {materialFunction=&MakeFloat2}.Result
    Alpha=NamedRerouteUsage()
   )) {materialFunction=&BreakOutFloat2Components}
  n93 = NamedRerouteDeclaration(None=@n92.R) {name=U_Scale, nodeColor=<rgba 1 0 0.3516 1>}
  n94 = NamedRerouteDeclaration(None=@n92.G) {name=V_Scale, nodeColor=<rgba 0 1 0 1>}
  n95 = Multiply(
   A=MaterialFunctionCall(UVs=NamedRerouteDeclaration(None=@n89) {name=RotatedUVs, nodeColor=<rgba 0 0.0156 1 1>}, Texture Scale=@n90.R) {materialFunction=&ScaleUVsByCenter}.UVs
   B=NamedRerouteDeclaration(None=AppendVector(A=@n93, B=@n94)) {name=UVScale, nodeColor=<rgba 0 0.2266 1 1>}
  )
  n96 = VectorParameter() {
   group=FNCSShield_Screen
   parameterName=UV_Bias
   sortPriority=20
   channelNames {a=Double_V_Bias, b=Double_U_Bias, g=Single_V_Bias, r=Single_U_Bias}
   defaultValue=<rgba -0.3167 -0.0233 -0.5087 -0.057>
  }
  n97 = MaterialFunctionCall(X=@n96.Single_U_Bias, Y=@n96.Single_U_Bias) {materialFunction=&MakeFloat2}
  n98 = MaterialFunctionCall(X=@n96.Double_U_Bias, Y=@n96.Double_U_Bias) {materialFunction=&MakeFloat2}
  n99 = NamedRerouteUsage()
  NamedRerouteDeclaration(None=Add(
    A=NamedRerouteDeclaration(None=@n95) {name=Rotated+ScaledUVs, nodeColor=<rgba 0 0.8828 1 1>}
    B=LinearInterpolate(A=@n97.Result, B=@n98.Result, Alpha=@n99)
   )) {name=DigitsWrappedUVs, nodeColor=<rgba 0.3359 0 1 1>}
  n100 = NamedRerouteUsage()
  n101 = Constant()
  n102 = Divide(A=Constant() {r=10}) {constB=11}
  n103 = MaterialFunctionCall(Float2=@n100) {materialFunction=&BreakOutFloat2Components}
  n104 = Constant() {r=0.5}
  n105 = SphereMask(A=@n103.R, B=@n104, Radius=@n104, Hardness=Constant() {r=1})
  n106 = Constant() {r=10}
  n107 = Divide(A=Floor(None=Modulo!(A=@n24, B=@n106))) {constB=11}
  n108 = SphereMask(A=@n103.R, B=Constant() {r=2.5}, Radius=Constant() {r=0.5}, Hardness=Constant() {r=1})
  n109 = Divide(A=@n24) {constB=10}
  n110 = Constant() {r=10}
  n111 = Divide(A=Floor(None=Modulo(A=@n109, B=@n110))) {constB=11}
  n112 = NamedRerouteUsage()
  n113 = Constant() {r=1.5}
  n114 = Constant() {r=0.5}
  n115 = Constant() {r=1}
  n116 = Add(
   A=Add(A=Multiply(A=@n102, B=@n105), B=Multiply(A=@n107, B=@n108))
   B=Multiply(A=LinearInterpolate(A=@n107, B=@n111, Alpha=@n112), B=SphereMask(A=@n103.R, B=@n113, Radius=@n114, Hardness=@n115))
  )
  NamedRerouteDeclaration(None=Add(A=@n100, B=AppendVector(A=@n101, B=@n116))) {name=DigitsUV, nodeColor=<rgba 0 0.9531 1 1>}
  n117 = NamedRerouteUsage()
  n118 = CurveAtlasRowParameter(CurveTime=NamedRerouteUsage()) {
   atlas=&CA_JumboTron
   curve=/Game/Accessories/FORT_Backpacks/Backpack_M_MED_Jumbotron_S20/Textures/C_JumboTron_BurstBG
   group=FNCSShield_Screen_GraphicsCurve
   parameterName=Burst
   sortPriority=0
  }
  NamedRerouteDeclaration(None=TextureSample!(
    UVs=MaterialFunctionCall(UVs=@n117, Texture Scale=@n118.R) {materialFunction=&ScaleUVsByCenter}.UVs
    Tex=@n59
    Apply View MipBias
   ) {samplerSource=SSM_Clamp_WorldGroupSettings, samplerType=SAMPLERTYPE_LinearColor}.G) {name=CelebrationBurst, nodeColor=<rgba 1 0.7266 0 1>}
  n119 = TextureSampleParameter2D(UVs=NamedRerouteUsage()) {
   group=FNCSShield_Screen
   parameterName=DigitScreen BG Texture
   samplerType=SAMPLERTYPE_Masks
   sortPriority=10
   texture=&T_White
  }
  n120 = VectorParameter() {group=FNCSShield_Screen, parameterName=DigitScreenBGTexture_CenterColor, sortPriority=20, defaultValue=<rgba 0 0 0 1>}
  n121 = NamedRerouteUsage()
  n122 = TextureSampleParameter2D(UVs=TextureCoordinate() {coordinateIndex=1}) {
   group=FNCSShield_Screen
   parameterName=Screen InnerMask
   samplerType=SAMPLERTYPE_Masks
   sortPriority=10
   texture=&T_BlackMask
  }
  n123 = SmoothStep(
   Min=ScalarParameter() {group=FNCSShield_Screen, parameterName=InnerMaskMin, sliderMax=1, sortPriority=30}
   Max=ScalarParameter() {defaultValue=1, group=FNCSShield_Screen, parameterName=InnerMaskMax, sliderMax=1, sortPriority=30}
   Value=Power(Base=@n122.R, Exp=ScalarParameter() {defaultValue=0.1, group=FNCSShield_Screen, parameterName=InnerMaskPower, sortPriority=30})
  )
  n124 = ScalarParameter() {defaultValue=1, group=FNCSShield_Screen, parameterName=InnerMaskOpacity, sliderMax=1, sortPriority=30}
  n125 = NamedRerouteUsage()
  n126 = Constant() {r=0.5}
  n127 = ScalarParameter() {defaultValue=1.25, group=FNCSShield_Screen, parameterName=BG_SphereMsk_Size, sortPriority=30}
  n128 = ScalarParameter() {defaultValue=0.25, group=FNCSShield_Screen, parameterName=BG_SphereMsk_Falloff, sortPriority=30}
  n129 = StaticSwitchParameter(
   True=NamedRerouteDeclaration(None=Multiply(A=@n123, B=@n124)) {name=InnerMask, nodeColor=<rgba 0 0.1562 1 1>}
   False=Power(Base=SphereMask(A=@n125, B=@n126, Radius=@n127, Hardness=@n128), Exp=Constant() {r=1.5})
  ) {defaultValue=true, desc=Ture: InnerMask False: SphereMask, group=FNCSShield_Screen, parameterName=Use InnerMask/SphereMask, sortPriority=2}
  n130 = LinearInterpolate(
   A=VectorParameter() {group=FNCSShield_Screen, parameterName=BGColorOuter, sortPriority=20, defaultValue=<rgba 1 1 1 1>}.RGB
   B=VectorParameter() {group=FNCSShield_Screen, parameterName=BGColorInner, sortPriority=20, defaultValue=<rgba 0 0 0 1>}.RGB
   Alpha=@n129
  )
  n131 = Multiply(A=OneMinus(None=@n129), B=NamedRerouteUsage())
  NamedRerouteDeclaration(None=StaticSwitchParameter(True=LinearInterpolate(A=@n119.RGB, B=@n120.RGB, Alpha=@n121), False=Add(A=@n130, B=@n131)) {group=FNCSShield_Screen, parameterName=Use BackGroundTextureDirectly, sortPriority=2}) {name=Screen Background, nodeColor=<rgba 1 0 0.5625 1>}
  NamedRerouteDeclaration(None=Step(X=@n24) {constX=10, constY=1}) {name=IfScoreOverZero, nodeColor=<rgba 0.2188 1 0 1>}
  n132 = TextureSampleParameter2D(UVs=@n1) {
   group=FNCSShield_Screen
   parameterName=Screen Normal
   samplerType=SAMPLERTYPE_Normal
   sortPriority=10
   texture=&FlatNormalBlue
  }
  NamedRerouteDeclaration(None=@n132.RGB) {name=outNormalMap, nodeColor=<rgba 0.4766 1 0 1>}
  n133 = NamedRerouteUsage()
  n134 = CurveAtlasRowParameter(CurveTime=NamedRerouteUsage()) {
   atlas=&CA_JumboTron
   curve=/Game/Accessories/FORT_Backpacks/Backpack_M_MED_Jumbotron_S20/Textures/C_JumboTron_Ribbons
   defaultValue=1
   group=FNCSShield_Screen_GraphicsCurve
   parameterName=Ribbons
   sortPriority=0
  }
  NamedRerouteDeclaration(None=TextureSample(
    UVs=MaterialFunctionCall(UVs=@n133, Texture Scale=@n134.R) {materialFunction=&ScaleUVsByCenter}.UVs
    Tex=@n59
   ) {samplerSource=SSM_Clamp_WorldGroupSettings, samplerType=SAMPLERTYPE_LinearColor}.B) {name=CelebrationRibbons, nodeColor=<rgba 1 0.7266 0 1>}
  n135 = Divide(A=Time(), B=ScalarParameter() {defaultValue=3, group=FNCSShield_Screen, parameterName=DebugTransitionLength, sortPriority=30})
  n136 = MaterialFunctionCall(
   TimestampA=ScalarParameter() {group=FNCSShield_Screen, parameterName=Timestamp, sortPriority=30}
   TransitionLength_In=ScalarParameter() {defaultValue=3, group=FNCSShield_Screen, parameterName=TransitionLength, sortPriority=30}
  ) {materialFunction=&MF_Timestamp_InOut_01}
  NamedRerouteDeclaration(None=StaticSwitchParameter(True=Frac(None=@n135), False=VertexInterpolator(VS=@n136.Linear).PS) {group=FNCSShield_Screen_Graphics, parameterName=DebugModeOn, sortPriority=0}) {name=TimeStamp, nodeColor=<rgba 0.125 1 0 1>}
 ]}
