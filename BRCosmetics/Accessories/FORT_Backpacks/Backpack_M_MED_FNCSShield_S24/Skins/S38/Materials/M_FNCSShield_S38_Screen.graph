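# Screen material for the Season 38 skin of the FNCS Shield backpack cosmetic.
# The S37 screen material function supplies the material attributes, so this asset holds no screen logic of its own.
# That result feeds the shared 2023 character PostFX function through its Material, M and Fresnel inputs.
# It is Default Lit, uses material attributes, and supports skeletal meshes, Nanite and mesh deformers.
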Material /BRCosmetics/Accessories/FORT_Backpacks/Backpack_M_MED_FNCSShield_S24/Skins/S38/Materials/M_FNCSShield_S38_Screen
graph {
 properties {
  bUseMaterialAttributes=true
  bUsedWithMeshDeformer=true
  bUsedWithNanite=true
  bUsedWithSkeletalMesh=true
  pixelDepthOffsetMode=PDOM_Legacy
  usedShadingModels=Default Lit
 }
 nodes [
  n1 = MaterialFunctionCall() {materialFunction=/BRCosmetics/Accessories/FORT_Backpacks/Backpack_M_MED_FNCSShield_S24/Materials/MF_FNCSShield_S37_Screen}
  MaterialFunctionCall(Material=@n1."Material (MA)", M=@n1."Material (MA)", Fresnel=@n1."Material (MA)") {materialFunction=&MF_CharacterParent_2023_PostFX}
 ]
}
